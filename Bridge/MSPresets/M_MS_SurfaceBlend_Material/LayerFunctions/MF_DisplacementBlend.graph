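# Material function for a three-layer Megascans surface blend material that outputs displacement, with a static switch between individual and unified displacement values.
# In individual mode, each Base, Middle and Top height is remapped by a Max parameter and scaled by an Amount through MF_Displacement.
# The results blend by Base/Middle Alpha, then Top Alpha.
# Otherwise, the input material attributes' WorldDisplacement red channel goes through MF_Displacement with one Unified Displacement scale.
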MaterialFunction /Bridge/MSPresets/M_MS_SurfaceBlend_Material/LayerFunctions/MF_DisplacementBlend
graph {
 properties {bExposeToLibrary=true}
 nodes [
  n1 = ScalarParameter() {defaultValue=1, group=06 - Displacement, parameterName=Top Layer Displacement Max, sortPriority=5}
  n2 = FunctionInput() {inputName=Top Layer Height, inputType=FunctionInput_Scalar, sortPriority=2}
  n3 = MaterialFunctionCall!(
   Height=ComponentMask(None=LinearInterpolate(B=@n1, Alpha=@n2)) {r=true}
   Scale=ScalarParameter() {defaultValue=1, group=06 - Displacement, parameterName=Top Layer Displacement Amount, sortPriority=4}
  ) {materialFunction=&/MF_Displacement}
  n4 = LinearInterpolate(
   B=ScalarParameter() {defaultValue=1, group=06 - Displacement, parameterName=Middle Layer Displacement Max, sortPriority=3}
   Alpha=FunctionInput() {inputName=Middle Layer Height, inputType=FunctionInput_Scalar, sortPriority=1}
  )
  n5 = LinearInterpolate(
   B=ScalarParameter() {defaultValue=1, group=06 - Displacement, parameterName=Base Layer Displacement Max, sortPriority=1}
   Alpha=FunctionInput() {inputName=Base Layer Height, inputType=FunctionInput_Scalar}
  )
  n6 = LinearInterpolate(
   A=MaterialFunctionCall(
    Height=ComponentMask(None=@n4) {r=true}
    Scale=ScalarParameter() {defaultValue=1, group=06 - Displacement, parameterName=Middle Layer Displacement Amount, sortPriority=2}
   ) {materialFunction=&MF_Displacement}.Result
   B=MaterialFunctionCall(
    Height=ComponentMask(None=@n5) {r=true}
    Scale=ScalarParameter() {defaultValue=1, group=06 - Displacement, parameterName=Base Layer Displacement Amount}
   ) {materialFunction=&MF_Displacement}.Result
   Alpha=FunctionInput() {inputName=Base/Middle Alpha, inputType=FunctionInput_Scalar, sortPriority=3}
  )
  n7 = FunctionInput() {inputName=Top Alpha, inputType=FunctionInput_Scalar, sortPriority=4}
  n8 = ComponentMask(None=GetMaterialAttributes(None=FunctionInput() {inputName=MA, inputType=FunctionInput_MaterialAttributes, sortPriority=5}) {attributeGetTypes [#1]}.WorldDisplacement) {r=true}
  n9 = ScalarParameter() {defaultValue=1, group=06 - Displacement, parameterName=Unified Displacement}
  FunctionOutput(None=StaticSwitchParameter(
    True=LinearInterpolate(A=@n3.Result, B=@n6, Alpha=@n7)
    False=MaterialFunctionCall(Height=@n8, Scale=@n9) {materialFunction=&MF_Displacement}.Result
   ) {group=00 - Global, parameterName=Use Individual DIsplacement Values, sortPriority=4}) {outputName=Displacement}
 ]
}
