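# MF_NormalCorrection is a Megascans surface blend layer function that rotates and scales a tangent-space normal.
# The normal's XY is multiplied by Strength, then fed as UVs to CustomRotator, which turns it around zero by the negated Rotation Angle, in 0-1 units.
# The rotated XY is appended to the untouched Z and output as Normal, with no renormalization.
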MaterialFunction /Bridge/MSPresets/M_MS_SurfaceBlend_Material/LayerFunctions/MF_NormalCorrection
graph {nodes [
  n1 = FunctionInput(Preview=Constant3Vector() {constant=<rgba 0 0 1 0>}) {inputName=Tangent Normal}
  n2 = Multiply(A=ComponentMask(None=@n1) {g=true, r=true}, B=FunctionInput() {inputName=Strength, inputType=FunctionInput_Scalar, sortPriority=2})
  n3 = Constant()
  n4 = Multiply(A=FunctionInput() {inputName=Rotation Angle, inputType=FunctionInput_Scalar, sortPriority=1}) {constB=-1}
  FunctionOutput(None=AppendVector(
    A=MaterialFunctionCall(UVs=@n2, Rotation Center=@n3, "Rotation Angle (0-1)"=@n4) {materialFunction=&CustomRotator}.Rotated Values
    B=ComponentMask(None=@n1) {b=true}
   )) {outputName=Normal, sortPriority=1}
 ]}
